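# UI sound cue that, by its name, plays when the player confirms exiting a match from the in-game menu.
# A single wave player plays a reused front-end Social accept click, routed to the application tray notification sound class.
# A concurrency override allows one instance at a time, stopping the oldest, and virtualization is disabled.
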SoundCue /Game/Sounds/Fort_UI_Sounds/Gameplay/Elimination/In_Game_Menu/UI_Exit_Match_confirm_Cue
root=WavePlayer_0
nodes [{id=WavePlayer_0, properties {soundWaveAssetPtr=/Game/Sounds/Fort_UI_Sounds/FrontEnd/Social/Fort_UI_Social_Accept_Click_08_b}}]
properties {
 bOverrideConcurrency=true
 maxDistance=2097152
 soundClassObject=/Game/Sounds/FortSoundClassesAndModes/Fort_Application_TrayNotification
 virtualizationMode=Disabled
 volumeMultiplier=1
 concurrencyOverrides {maxCount=1, resolutionRule=StopOldest}
}
